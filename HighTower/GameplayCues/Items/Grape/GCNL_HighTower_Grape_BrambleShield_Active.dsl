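; Looping gameplay cue for the active Grape Bramble Shield item, handling its audio on the affected Fortnite pawn.
; On application it caches the pawn, fades in a layered audio component, attaches it to the pawn's root and adds a sound library for the pawn.
; Each tick it maps pawn velocity to a normalized "Speed" audio parameter.
; On removal it fades the audio out and removes the sound library.
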(event EventTick
  (params (DeltaSeconds :type Float%20%28single-precision%29))
  (bind _fp (Variables|Default|GetFP)
    (:then
      (Class|FortLayeredAudioComponent|SetFloatParameterExt (Variables|Default|GetFortLayeredAudio) "Speed" (Math|Float|MapRangeClamped (Math|Vector|VectorLength (Transformation|GetVelocity _fp)) 1e+02 1e+03 0.0 1.0)))
    (:else)))

(event EventOnApplication(EitherVFX-preferthisone)
  (params (MyTarget :type Actor%20Object%20Reference) (Parameters :type Gameplay%20Cue%20Parameters%20Structure%20%28by%20ref%29) (ParticleComponents :type Array%20of%20FXSystem%20Component%20Object%20References) (AudioComponents :type Array%20of%20Audio%20Component%20Object%20References) (BurstCameraShakeInstance :type Camera%20Shake%20Base%20Object%20Reference) (BurstDecalInstance :type Decal%20Actor%20Object%20Reference))
  (bind _fortlayeredaudio (Variables|Default|GetFortLayeredAudio))
  (bind _asfort_pawn (Utilities|Casting|CastToFortPawn MyTarget))
  (bind _output_get (Variables|Default|SetFP _asfort_pawn))
  (Audio|Components|Audio|FadeIn _fortlayeredaudio 1.0 1.0 0.0 "Linear" :!class /Script/Engine.AudioComponent)
  (Transformation|AttachComponentToComponent _fortlayeredaudio (Variables|/Script/Engine.Actor|GetRootComponent _output_get) "None" "SnapToTarget" "SnapToTarget" "SnapToTarget" true)
  (SoundLibrary|AddLibrary (WorldSubsystems|GetSoundLibrarySubsystem (Variables|Default|GetFP)) (Variables|Default|GetFP) 0 :!class /Script/SoundLibrary.SoundLibrarySubsystem))

(event EventOnRemoval(EitherVFX-preferthisversion)
  (params (MyTarget :type Actor%20Object%20Reference) (Parameters :type Gameplay%20Cue%20Parameters%20Structure%20%28by%20ref%29) (ParticleComponents :type Array%20of%20FXSystem%20Component%20Object%20References) (AudioComponents :type Array%20of%20Audio%20Component%20Object%20References) (BurstCameraShakeInstance :type Camera%20Shake%20Base%20Object%20Reference) (BurstDecalInstance :type Decal%20Actor%20Object%20Reference))
  (Audio|Components|Audio|FadeOut (Variables|Default|GetFortLayeredAudio) 1.0 0.0 "Linear" :!class /Script/Engine.AudioComponent)
  (SoundLibrary|RemoveLibrary (WorldSubsystems|GetSoundLibrarySubsystem (Variables|Default|GetFP)) (Variables|Default|GetFP) 0 :!class /Script/SoundLibrary.SoundLibrarySubsystem))

(fn ConstructionScript)
